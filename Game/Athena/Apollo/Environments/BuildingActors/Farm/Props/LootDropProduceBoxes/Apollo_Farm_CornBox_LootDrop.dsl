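; Apollo farm corn box prop Blueprint, named as a loot-drop produce box, that removes itself when the actor supporting it is destroyed.
; A linked-actor-destroyed event from the FortLinkToActor component calls DestroyActor, standing in for default prop linking, which the author notes can be inaccurate.
; The construction script is empty, and no loot logic appears in this graph.
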(event OnLinkedActorDestroyed(FortLinkToActor)
  (params (DamageCauser :type Actor%20Object%20Reference))
  ;[1 1 1] Sometimes the default actor linking for props is not linking as accurately as we would like. Using the link to actor component gets a better result. Destroy this actor if its support is broken.
  (Actor|DestroyActor))

(fn ConstructionScript)
